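; Manager Blueprint that reserves living-world AI density for PistolCanary NPC Heroes so they have room to spawn.
; On BeginPlay, the authority waits for playlist data, then registers a density reservoir sized by a scalable-float hero count.
; If registration succeeds, it listens on the global event router and, on one hero-manager event, unregisters the reservoir and requests event generation.
; If registration fails, it logs a warning that Event Heroes will probably not spawn.
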(event EventBeginPlay
  (bind _sf_npcherocount (Variables|Internal|GetSF_NPCHeroCount))
  (bind _returnvalue (ScalableFloat|GetValueatLevel _sf_npcherocount 1.0))
  (bind _returnvalue_2 (Class|FortAthenaLivingWorldManager|GetLivingWorldManager))
  (Networking|SwitchHasAuthority
    (:Authority
      (Development|PrintString "@@@ Event BeginPlay [BP_PistolCanary_Global_NPCReserve]: Authority - CALLED!" false (Variables|Debug|GetbDebugReserve))
      (Class|FortAsyncActionCurrentPlaylistReady|PlaylistDataReadyAsync
        (:OnReady
          (bind _returnvalue_3 (Class|FortAthenaLivingWorldManager|RegisterDensityReservoir (Class|FortAthenaLivingWorldManager|GetLivingWorldManager) self _returnvalue 4.5e+03))
          (if _returnvalue_3
            (bind (_asynctaskproxy _payload) (GameplayEventRouter|StartListeningToEvent (GameplayEventRouter|GetGlobalEventRouter) "/PistolCanary/Gameplay/EventRouter/PistolCanary_NPCHeroManager_Events.PistolCanary_NPCHeroManager_Events")
              (:then
                (Variables|Internal|SetAsyncTaskEventListener _asynctaskproxy))
              (:OnEventReceived
                (switch Utilities|FlowControl|Switch|SwitchonPistolCanary_NPCHeroManager_Events_Enum (Utilities|Struct|BreakPistolCanaryNPCHeroManagerEvents _payload)
                  (:NewEnumerator5
                    (Class|FortAthenaLivingWorldManager|UnregisterDensityReservoir _returnvalue_2 self)
                    (Class|FortAthenaLivingWorldManager|RequestEventGeneration _returnvalue_2)))))
            (else
              (Fortnite|FortShippingLog self (Utilities|String|Append "BP_PistolCanary_Global_NPCReserve: Attempted to register " (Utilities|String|ToString(Float) _returnvalue) " AI agents in the Density Register and failed!  Event Heroes will probably NOT be able to spawn here!") true)
              (Development|PrintString "@@@@@@ Event Begin play [BP_PistolCanary_Global_NPCReserve]: FAILED to secure spawn density reservoir!" false (Variables|Debug|GetbDebugReserve)))))))))

(fn ConstructionScript)
