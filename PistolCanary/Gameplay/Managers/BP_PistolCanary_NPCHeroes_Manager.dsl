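; Server-side manager actor that drives the PistolCanary NPC heroes' spawning, staging and scripted behaviours.
; Once playlist data is ready, it listens on the global event router and dispatches each event type to a handler, resolving heroes through a hero-kind map.
; A one-time spawn clears nearby sprite pawns, swaps point providers and forces Living World generation.
; Other handlers re-leash heroes to staging points, trigger supers, set speech state or kill heroes.
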(event EventBeginPlay
  (Networking|SwitchHasAuthority
    (:Authority
      (Class|FortAsyncActionCurrentPlaylistReady|PlaylistDataReadyAsync
        (:OnReady
          (Development|PrintString "@@@ Event BeginPlay [BP_PistolCanary_NPCHeroes_Manager]: Authority - CALLED!" false (Variables|Debug|GetbDebugManager))
          (CallFunction|InitializeManager)
          (bind (_asynctaskproxy _payload) (GameplayEventRouter|StartListeningToEvent (GameplayEventRouter|GetGlobalEventRouter) "/PistolCanary/Gameplay/EventRouter/PistolCanary_NPCHeroManager_Events.PistolCanary_NPCHeroManager_Events")
            (:then
              (Variables|Internal|SetAsyncTask_EventListener _asynctaskproxy))
            (:OnEventReceived
              (switch Utilities|FlowControl|Switch|SwitchonPistolCanary_NPCHeroManager_Events_Enum (Utilities|Struct|BreakPistolCanaryNPCHeroManagerEvents _payload)
                (:NewEnumerator0
                  (CallFunction|OnSpawnEventReceived))
                (:NewEnumerator1
                  (CallFunction|OnStagingEventReceived (.TargetNPCHero_5_FF7E79044B618402842D909984D0927E (Utilities|Struct|BreakPistolCanaryNPCHeroManagerEvents _payload))))
                (:NewEnumerator2
                  (CallFunction|OnSuperEventReceived (.TargetNPCHero_5_FF7E79044B618402842D909984D0927E (Utilities|Struct|BreakPistolCanaryNPCHeroManagerEvents _payload))))
                (:NewEnumerator3
                  (CallFunction|OnStagingEventReceived (.TargetNPCHero_5_FF7E79044B618402842D909984D0927E (Utilities|Struct|BreakPistolCanaryNPCHeroManagerEvents _payload)) 1))
                (:NewEnumerator4
                  (CallFunction|OnStagingEventReceived (.TargetNPCHero_5_FF7E79044B618402842D909984D0927E (Utilities|Struct|BreakPistolCanaryNPCHeroManagerEvents _payload)) 2))
                (:NewEnumerator6
                  (CallFunction|OnStagingEventReceived (.TargetNPCHero_5_FF7E79044B618402842D909984D0927E (Utilities|Struct|BreakPistolCanaryNPCHeroManagerEvents _payload)) 3))
                (:NewEnumerator7
                  (CallFunction|OnStagingEventReceived (.TargetNPCHero_5_FF7E79044B618402842D909984D0927E (Utilities|Struct|BreakPistolCanaryNPCHeroManagerEvents _payload)) 4))
                (:NewEnumerator8
                  (CallFunction|OnStagingEventReceived (.TargetNPCHero_5_FF7E79044B618402842D909984D0927E (Utilities|Struct|BreakPistolCanaryNPCHeroManagerEvents _payload)) 5))
                (:NewEnumerator9
                  (CallFunction|OnSpeechBubbleEventReceived (.TargetNPCHero_5_FF7E79044B618402842D909984D0927E (Utilities|Struct|BreakPistolCanaryNPCHeroManagerEvents _payload)) true))
                (:NewEnumerator10
                  (CallFunction|OnSpeechBubbleEventReceived (.TargetNPCHero_5_FF7E79044B618402842D909984D0927E (Utilities|Struct|BreakPistolCanaryNPCHeroManagerEvents _payload))))
                (:NewEnumerator11
                  (CallFunction|OnHeroRemovedEventReceived (.TargetNPCHero_5_FF7E79044B618402842D909984D0927E (Utilities|Struct|BreakPistolCanaryNPCHeroManagerEvents _payload))))
                (:NewEnumerator12
                  (CallFunction|OnStagingEventReceived (.TargetNPCHero_5_FF7E79044B618402842D909984D0927E (Utilities|Struct|BreakPistolCanaryNPCHeroManagerEvents _payload)) 6))
                (:NewEnumerator13
                  (CallFunction|OnStagingEventReceived (.TargetNPCHero_5_FF7E79044B618402842D909984D0927E (Utilities|Struct|BreakPistolCanaryNPCHeroManagerEvents _payload)) 7))
                (:NewEnumerator14
                  (CallFunction|OnStagingEventReceived (.TargetNPCHero_5_FF7E79044B618402842D909984D0927E (Utilities|Struct|BreakPistolCanaryNPCHeroManagerEvents _payload)) 8))
                (:NewEnumerator15
                  (CallFunction|OnStagingEventReceived (.TargetNPCHero_5_FF7E79044B618402842D909984D0927E (Utilities|Struct|BreakPistolCanaryNPCHeroManagerEvents _payload)) 9))
                (:NewEnumerator16
                  (CallFunction|OnMomentOfCalmEnabled))))))))))

(event Custom|OnSpawnEventReceived
  (Development|PrintString "@@@@@@ OnSpawnEventReceived [BP_PistolCanary_NPCHeroes_Manager]: CALLED!" false (Variables|Debug|GetbDebugManager))
  (if (not (Variables|Internal|GetbHasSpawnedHeroes))
    (Variables|Internal|SetbHasSpawnedHeroes true)
    (GameplayEventRouter|BroadcastEvent (GameplayEventRouter|GetGlobalEventRouter) (Utilities|Struct|MakePistolCanaryNPCHeroManagerEvents "NewEnumerator5") 0)
    (bind _outfortpawns (Fortnite|GetFortPawnsInRange (Transformation|GetActorLocation) 2e+04 true)) ; 18740.0
    (for _array_element _outfortpawns
      (Utilities|Casting|CastToNPC_Pawn_Irwin_Simple_Sprite _array_element)
      (Actor|DestroyActor _array_element))
    (for _array_element_2 (Variables|Tunable|GetPointProviderArray_NPCHeroes)
      (Utilities|IsValid _array_element_2
        (:Is%20Valid
          (Class|FortAthenaLivingWorldPointProviderInterface|EnablePointProvider_DEPRECATED _array_element_2)
          (Development|PrintString (Utilities|String|Append "@@@@@@ OnSpawnEventReceived, Enable Point Provider: " (Utilities|GetDisplayName _array_element_2)) false (Variables|Debug|GetbDebugManager)))))
    (bind _outactors (Fortnite|GetAllActorsOfClassinRange "/Script/LagerRuntime.FortAthenaLivingWorldStaticPointProvider" (Transformation|GetActorLocation) 1.874e+04))
    (if (Utilities|Array|IsNotEmpty _outactors)
      (for _array_element_3 _outactors
        (if (== (Class|FortAthenaLivingWorldPointProviderInterface|GetFiltersTags _array_element_3) (Variables|Internal|GetSpriteFilterTag))
          (Class|FortAthenaLivingWorldPointProviderInterface|DisablePointProvider _array_element_3) ; disable sprites
        ))
      (Class|FortAthenaLivingWorldManager|RequestEventGeneration (Class|FortAthenaLivingWorldManager|GetLivingWorldManager)) ; forces Living World to update / spawn NPCs without\nwaiting for 5 seconds
    )))

(event Custom|OnStagingEventReceived
  (params (TargetHeroKind :type PistolCanary_NPCHeroManager_HeroKind_Enum%20Enum :default "NewEnumerator0") (StagingLocIndex :type Integer))
  (bind _herokindmap (Variables|Internal|GetHeroKindMap))
  (bind (_value _returnvalue) (Utilities|Map|Find _herokindmap TargetHeroKind))
  (bind _targetpointarray_npcheroes_staginglocs (Variables|Tunable|GetTargetPointArray_NPCHeroes_StagingLocs))
  (bind _output (Utilities|Array|Get(acopy) _targetpointarray_npcheroes_staginglocs StagingLocIndex))
  (Development|PrintString "@@@@@@ OnStagingEventReceived [BP_PistolCanary_NPCHeroes_Manager]: CALLED!" false (Variables|Debug|GetbDebugManager))
  (if _returnvalue
    (Utilities|IsValid _value
      (:Is%20Valid
        (Utilities|IsValid _output
          (:Is%20Valid
            (bind _asfort_athena_aibot_controller (Utilities|Casting|CastToFortAthenaAIBotController (Pawn|GetController _value)))
            (Utilities|IsValid (Actor|GetComponentbyClass _asfort_athena_aibot_controller "/Script/FortniteAI.FortAthenaNpcPatrollingComponent")
              (:Is%20Valid
                (Class|FortAthenaNpcPatrollingComponent|RemovePatrolPath (Actor|GetComponentbyClass _asfort_athena_aibot_controller "/Script/FortniteAI.FortAthenaNpcPatrollingComponent")))
              (:Is%20Not%20Valid
                _))
            (Class|BPFLAICommonFunctions|SetRadialLeash _asfort_athena_aibot_controller (Transformation|GetActorLocation _output) (ScalableFloat|GetValueatLevel (Variables|Internal|GetSF_PointLeash_InnerRadius) 1.0) (ScalableFloat|GetValueatLevel (Variables|Internal|GetSF_PointLeash_OuterRadius) 1.0))))))))

(event Custom|OnSuperEventReceived
  (params (TargetHeroKind :type PistolCanary_NPCHeroManager_HeroKind_Enum%20Enum :default "NewEnumerator0"))
  (bind _herokindmap (Variables|Internal|GetHeroKindMap))
  (bind (_value _returnvalue) (Utilities|Map|Find _herokindmap TargetHeroKind))
  (Development|PrintString "@@@@@@ OnSuperEventReceived [BP_PistolCanary_NPCHeroes_Manager]: CALLED!" false (Variables|Debug|GetbDebugManager))
  (if _returnvalue
    (Utilities|IsValid _value
      (:Is%20Valid
        (if (GameplayTags|HasMatchingGameplayTag _value "(TagName=\"AISpawnerData.Type.Dev07\")")
          (Ability|SendGameplayEventtoActor _value "(TagName=\"Event.Hero01.Action01\")")
          (elif (GameplayTags|HasMatchingGameplayTag _value "(TagName=\"AISpawnerData.Type.Dev08\")")
            (Ability|SendGameplayEventtoActor _value "(TagName=\"Event.Hero02.Action01\")")
            (elif (GameplayTags|HasMatchingGameplayTag _value "(TagName=\"AISpawnerData.Type.Dev09\")")
              (Ability|SendGameplayEventtoActor _value "(TagName=\"Event.Hero03.Action01\")"))))))))

(event Custom|OnSpeechBubbleEventReceived
  (params (Target%20Hero%20Kind :type PistolCanary_NPCHeroManager_HeroKind_Enum%20Enum :default "NewEnumerator0") (Enable%20Speech%20Bubble :type Boolean))
  (bind _herokindmap (Variables|Internal|GetHeroKindMap))
  (bind (_value _returnvalue) (Utilities|Map|Find _herokindmap Target%20Hero%20Kind))
  (Development|PrintString "@@@@@@ OnSpeechBubbleEventReceived [BP_PistolCanary_NPCHeroes_Manager]: CALLED!" false (Variables|Debug|GetbDebugManager))
  (if _returnvalue
    (Utilities|IsValid _value
      (:Is%20Valid
        (Variables|/PistolCanary/NPCs/Heroes/Pawns/BP_Pawn_PistolCanary_Heroes.BP_Pawn_PistolCanary_Heroes_C|SetbSpeechBubblesEnabled Enable%20Speech%20Bubble _value)))))

(event Custom|OnHeroRemovedEventReceived
  (params (TargetHeroKind :type PistolCanary_NPCHeroManager_HeroKind_Enum%20Enum :default "NewEnumerator0"))
  (bind _herokindmap (Variables|Internal|GetHeroKindMap))
  (bind (_value _returnvalue) (Utilities|Map|Find _herokindmap TargetHeroKind))
  (Development|PrintString "@@@@@@ OnHeroRemovedEventReceived [BP_PistolCanary_NPCHeroes_Manager]: CALLED!" false (Variables|Debug|GetbDebugManager))
  (if _returnvalue
    (Utilities|IsValid _value
      (:Is%20Valid
        (FortPawn|ForceKill :self _value)))))

(event Custom|OnMomentOfCalmEnabled
  (Development|PrintString "@@@@@@ OnMomentOfCalmEnabled [BP_PistolCanary_NPCHeroes_Manager]: CALLED!" false (Variables|Debug|GetbDebugManager))
  (bind (_keypin _valuepin) (Utilities|Map|ForEachLoop(Map) (Variables|Internal|GetHeroKindMap)))
  (Utilities|IsValid _valuepin
    (:Is%20Valid
      (Variables|/PistolCanary/NPCs/Heroes/Pawns/BP_Pawn_PistolCanary_Heroes.BP_Pawn_PistolCanary_Heroes_C|SetbSpeechIsMomentOfCalm true _valuepin))))

(event Custom|Test_HeroManager_Spawn
  (Networking|SwitchHasAuthority
    (:Authority
      (Development|PrintString "@@@@@@ Test_HeroManager_Spawn [BP_PistolCanary_NPCHeroes_Manager]: CALLED!" false (Variables|Debug|GetbDebugManager))
      (GameplayEventRouter|BroadcastEvent (GameplayEventRouter|GetGlobalEventRouter) (Utilities|Struct|MakePistolCanaryNPCHeroManagerEvents "NewEnumerator0") 0))))

(event Custom|Test_HeroManager_MistRaven_Staging03
  (Networking|SwitchHasAuthority
    (:Authority
      (Development|PrintString "@@@@@@ Test_HeroManager_MistRaven_Staging01 [BP_PistolCanary_NPCHeroes_Manager]: CALLED!" false (Variables|Debug|GetbDebugManager))
      (GameplayEventRouter|BroadcastEvent (GameplayEventRouter|GetGlobalEventRouter) (Utilities|Struct|MakePistolCanaryNPCHeroManagerEvents "NewEnumerator4") 0))))

(event Custom|Test_HeroManager_AshenMagnus_Staging02
  (Networking|SwitchHasAuthority
    (:Authority
      (Development|PrintString "@@@@@@ Test_HeroManager_AshenMagnus_Staging02 [BP_PistolCanary_NPCHeroes_Manager]: CALLED!" false (Variables|Debug|GetbDebugManager))
      (GameplayEventRouter|BroadcastEvent (GameplayEventRouter|GetGlobalEventRouter) (Utilities|Struct|MakePistolCanaryNPCHeroManagerEvents "NewEnumerator3" "NewEnumerator2") 0))))

(event Custom|Test_HeroManager_PolarityWinn_Staging01
  (Networking|SwitchHasAuthority
    (:Authority
      (Development|PrintString "@@@@@@ Test_HeroManager_PolarityWinn_Staging01 [BP_PistolCanary_NPCHeroes_Manager]: CALLED!" false (Variables|Debug|GetbDebugManager))
      (GameplayEventRouter|BroadcastEvent (GameplayEventRouter|GetGlobalEventRouter) (Utilities|Struct|MakePistolCanaryNPCHeroManagerEvents "NewEnumerator1" "NewEnumerator3") 0))))

(event Custom|Test_HeroManager_MistRaven_Super
  (Networking|SwitchHasAuthority
    (:Authority
      (Development|PrintString "@@@@@@ Test_HeroManager_MistRaven_Super [BP_PistolCanary_NPCHeroes_Manager]: CALLED!" false (Variables|Debug|GetbDebugManager))
      (GameplayEventRouter|BroadcastEvent (GameplayEventRouter|GetGlobalEventRouter) (Utilities|Struct|MakePistolCanaryNPCHeroManagerEvents "NewEnumerator2") 0))))

(event Custom|Test_HeroManager_DisableBubbles
  (Networking|SwitchHasAuthority
    (:Authority
      (Development|PrintString "@@@@@@ Test_HeroManager_DisableBubbles [BP_PistolCanary_NPCHeroes_Manager]: CALLED!" false (Variables|Debug|GetbDebugManager))
      (bind (_keypin _valuepin) (Utilities|Map|ForEachLoop(Map) (Variables|Internal|GetHeroKindMap)))
      (Utilities|IsValid _valuepin
        (:Is%20Valid
          (GameplayEventRouter|BroadcastEvent (GameplayEventRouter|GetGlobalEventRouter) (Utilities|Struct|MakePistolCanaryNPCHeroManagerEvents "NewEnumerator10" _keypin) 0))))))

(event Custom|Test_HeroManager_RemoveHeroes
  (Networking|SwitchHasAuthority
    (:Authority
      (Development|PrintString "@@@@@@ Test_HeroManager_RemoveHeroes [BP_PistolCanary_NPCHeroes_Manager]: CALLED!" false (Variables|Debug|GetbDebugManager))
      (bind (_keypin _valuepin) (Utilities|Map|ForEachLoop(Map) (Variables|Internal|GetHeroKindMap)))
      (Utilities|IsValid _valuepin
        (:Is%20Valid
          (GameplayEventRouter|BroadcastEvent (GameplayEventRouter|GetGlobalEventRouter) (Utilities|Struct|MakePistolCanaryNPCHeroManagerEvents "NewEnumerator11" _keypin) 0))))))

(event Custom|Test_HeroManager_Calm
  (Networking|SwitchHasAuthority
    (:Authority
      (Development|PrintString "@@@@@@ TTest_HeroManager_Calm [BP_PistolCanary_NPCHeroes_Manager]: CALLED!" false (Variables|Debug|GetbDebugManager))
      (GameplayEventRouter|BroadcastEvent (GameplayEventRouter|GetGlobalEventRouter) (Utilities|Struct|MakePistolCanaryNPCHeroManagerEvents "NewEnumerator16") 0))))

(fn InitializeManager
  (Variables|Internal|SetbHasSpawnedHeroes false))

(fn ConstructionScript)
